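# Physics asset for the Carrot On A Stick two-handed melee pickaxe skeletal mesh, a cosmetic harvesting tool.
# It has a single box body on the Melee_TwoHanded_Handle bone, offset and rotated to enclose the whole weapon.
# No constraints exist, so the pickaxe simulates and collides as one rigid shape with no jointed or dangling parts.
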PhysicsAsset /BRCosmetics/Weapons/FORT_Melee/Pickaxe_Carrot_On_A_Stick/Meshes/SK_Pickaxe_Carrot_On_A_Stick_PhysicsAsset
bodies {Melee_TwoHanded_Handle [{shapeType=Box, extentX=98.4815, extentY=30.9745, extentZ=149.9052, center=<0.2102 4.0822 32.6979>, rotation=<rot -3.7757 -90 -0.1901>}]}
constraints []
